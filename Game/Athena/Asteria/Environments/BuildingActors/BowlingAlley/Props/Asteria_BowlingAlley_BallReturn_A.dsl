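; Blueprint for a bowling alley ball return prop that toggles its light visuals based on effects quality.
; The construction script calls ValidateLightVisuals, which reads the sg.EffectsQuality console variable and a preview level value.
; The spot light shows only at effects quality 3 or higher with preview level at most 1.
; The plane component is visible whenever the preview level is not 1, independent of effects quality.
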(fn ConstructionScript
  (CallFunction|ValidateLightVisuals))

(fn ValidateLightVisuals
  (bind _returnvalue (Development|GetConsoleVariableIntValue "sg.EffectsQuality"))
  (LocalVariables|Setl_lightVis (and (>= _returnvalue 3) (<= (Variables|Default|GetTestPreviewLevel) 1)))
  (Rendering|SetVisibility (Variables|Default|GetSpotLight) (LocalVariables|Getl_lightVis) false)
  (Rendering|SetVisibility (Variables|Default|GetPlane) (!= (Variables|Default|GetTestPreviewLevel) 1) false))
