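; Bowling alley table building prop that varies its mesh and color swatch, optionally picked from its world position.
; The construction script derives indices from averaged absolute coordinates, a transition distance and a modulo, then validates swatches before meshes.
; Swatches come from a palette object built from a data asset class, released on BeginPlay, and set building texture data.
; The chosen mesh is blocking-loaded into an added component, optionally inheriting base materials or texture overrides.
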(event EventBeginPlay
  (|CallParentFunction)
  (Variables|Default|SetcurrentPalette 0) ; Release Current Palette object
)

(event Custom|LoadMesh
  (bind _returnvalue (Utilities|LoadAssetBlocking (Utilities|Array|Get(acopy) (Variables|RandomMesh|GetMeshList) (Variables|RandomMesh|GetMeshIndex))))
  (bind _asstatic_mesh (Utilities|Casting|CastToStaticMesh _returnvalue))
  (Variables|Default|SetCurrentStaticMeshLoaded _asstatic_mesh)
  (Utilities|IsValid (Variables|Default|GetCurrentLoadedMeshComponent)
    (:Is%20Valid
      _)
    (:Is%20Not%20Valid
      (bind _returnvalue_2 (AddComponent|NODE_AddStaticMeshComponent-0))
      (Variables|Default|SetCurrentLoadedMeshComponent _returnvalue_2)))
  (Components|StaticMesh|SetStaticMesh (Variables|Default|GetCurrentLoadedMeshComponent) (Variables|Default|GetCurrentStaticMeshLoaded))
  (RandomMesh|ValidateMeshMaterials))

(fn GetSwatchArray
  (bind _swatchdataasset (Variables|RandomSwatch|GetSwatchDataAsset))
  (if (Utilities|IsValidClass _swatchdataasset)
    (Utilities|IsValid (Variables|Default|GetcurrentPalette)
      (:Is%20Valid
        _)
      (:Is%20Not%20Valid
        (bind _returnvalue (Game|ConstructObjectfromClass _swatchdataasset self))
        (Variables|Default|SetcurrentPalette _returnvalue)))
    (return (Variables|/Game/Environments/Apollo/Blueprints/RandomColor_Prop/RandomColor_Prop_Palette.RandomColor_Prop_Palette_C|GetSwatch_Palette (Variables|Default|GetcurrentPalette)) (Utilities|Array|Length (Variables|/Game/Environments/Apollo/Blueprints/RandomColor_Prop/RandomColor_Prop_Palette.RandomColor_Prop_Palette_C|GetSwatch_Palette (Variables|Default|GetcurrentPalette))))
    (else
      (Variables|Default|SetcurrentPalette 0))))

(fn RandomMeshByPosition
  (bind _meshlist (Variables|RandomMesh|GetMeshList))
  (bind _returnvalue (Utilities|Array|Length _meshlist))
  (if (> _returnvalue 0)
    (Variables|RandomMesh|SetMeshIndex (Math|Integer|%25%28Integer%29 (Math|Float|Round (/ (/ (Math|Float|float+float (Math|Float|Absolute(Float) (.x (Transformation|GetActorLocation))) (Math|Float|Absolute(Float) (.y (Transformation|GetActorLocation))) (Math|Float|Absolute(Float) (.z (Transformation|GetActorLocation)))) 3.0) (Variables|RandomMesh|GetMeshTransitionDistance))) _returnvalue))))

(fn RandomSwatchByPosition
  (bind (_color_swatches _swatch_count) (SwatchList|GetSwatchArray))
  (if (> _swatch_count 0)
    (Variables|RandomSwatch|SetSwatchIndex (Math|Integer|%25%28Integer%29 (Math|Float|Round (/ (/ (Math|Float|float+float (Math|Float|Absolute(Float) (.x (Transformation|GetActorLocation))) (Math|Float|Absolute(Float) (.y (Transformation|GetActorLocation))) (Math|Float|Absolute(Float) (.z (Transformation|GetActorLocation)))) 3.0) (Variables|RandomSwatch|GetSwatchTransitionDistance))) _swatch_count))))

(fn RunRandomization
  (Utilities|FlowControl|Sequence
    (:then_0
      (if (Variables|RandomMesh|GetRandomizeMeshByPosition)
        (RandomMesh|RandomMeshbyPosition)))
    (:then_1
      (if (Variables|RandomSwatch|GetRandomizeSwatchByPosition)
        (RandomSwatch|RandomSwatchbyPosition)))
    (:then_2
      (CallFunction|ValidateVisualsAll))))

(fn ShouldRandomize
  (return (and (< (Variables|RandomSwatch|GetSwatchIndex) 0) (< (Variables|RandomMesh|GetMeshIndex) 0))))

(fn ConstructionScript
  (|CallParentFunction)
  (if (Editor|InEditororPIE)
    _
    (elif (Default|ShouldRandomize)
      _
      (else
        (CallFunction|ValidateVisualsAll))))
  (RandomSwatch|RunRandomization))

(fn ValidateVisualsAll
  (Utilities|FlowControl|Sequence
    (:then_0
      (SwatchList|ValidateVisualsSwatches) ; Needs to be before mesh visual validation
    )
    (:then_1
      (MeshList|ValidateVisualsMeshes))))

(fn ValidateMeshMaterials
  (bind _currentloadedmeshcomponent (Variables|Default|GetCurrentLoadedMeshComponent))
  (switch Utilities|FlowControl|Switch|SwitchonRandomColorProp_MaterialApply_Enum (Variables|RandomMesh|GetMeshMaterialOverride)
    (:NewEnumerator1
      ;[1 1 1] Copy the base mesh materials to the newly load mesh
      (Utilities|IsValid (Variables|Default|GetCurrentLoadedMeshComponent)
        (:Is%20Valid
          (for (_array_element _array_index) (Rendering|Material|GetMaterials (Variables|Debug|GetStaticMeshComponent))
            (if (< _array_index (Rendering|Material|GetNumMaterials _currentloadedmeshcomponent))
              (Rendering|Material|SetMaterial _currentloadedmeshcomponent _array_index _array_element))))))
    (:NewEnumerator2
      (bind _outtexturedataarray (Building|GetBuildingTextureData))
      (if (Utilities|Array|IsValidIndex _outtexturedataarray)
        (Utilities|CopyDataOverridestoMesh (Variables|Default|GetCurrentLoadedMeshComponent) (Utilities|Array|Get(acopy) _outtexturedataarray))))))

(fn ValidateVisuals_Meshes
  (Utilities|FlowControl|Sequence
    (:then_0
      (if (== (Utilities|Array|Length (Variables|RandomMesh|GetMeshList)) 0)
        (return)))
    (:then_1
      (if (>= (Variables|RandomMesh|GetMeshIndex) (Utilities|Array|Length (Variables|RandomMesh|GetMeshList)))
        (Variables|RandomMesh|SetMeshIndex (- (Utilities|Array|Length (Variables|RandomMesh|GetMeshList)) 1))
        (elif (< (Variables|RandomMesh|GetMeshIndex) 0)
          (Variables|RandomMesh|SetMeshIndex 0))))
    (:then_2
      (CallFunction|LoadMesh))))

(fn ValidateVisuals_Swatches
  (bind (_color_swatches _swatch_count) (SwatchList|GetSwatchArray))
  ;[1 1 1] Comment
  (if (> _swatch_count 0)
    ;[1 1 1] Comment
    (if (Utilities|Array|IsValidIndex _color_swatches (Variables|RandomSwatch|GetSwatchIndex))
      ;[1 1 1] Comment
      (CallFunction|SetBuildingTextureData 0 (Utilities|Array|Get(acopy) _color_swatches (Variables|RandomSwatch|GetSwatchIndex)))
      (else
        ;[1 1 1] Comment
        (Variables|RandomSwatch|SetSwatchIndex (Math|Integer|SelectInt 0 (- _swatch_count 1) (< (Variables|RandomSwatch|GetSwatchIndex) 0)))))))
